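# Level sequence for PistolCanary Section 03, a player-POV cinematic in which a superhero NPC swoops in near the player mannequin and leaves again.
# Player, camera and NPC use attach tracks, while a moverOffset mesh holds position and rotates.
# The camera keys transform and focal length, with weighted template-sequence animations layered on its rotation.
# Two subsequences supply further content, and data layer tracks control section, tentacle and inner-spire world content.
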LevelSequence /PistolCanary/Sequence/PistolCanary_Section_03/PistolCanary_Section_03
bindings [
 CameraComponent {
  id=#1
  parent_id=#2
  possessed_class=CineCameraComponent
  tracks [
   {
    type=MovieSceneFloatTrack
    display_name=Current Focal Length
    sections [{
      type=MovieSceneFloatSection
      channels [None {
        type=MovieSceneScriptingFloatChannel
        default=20.4766
        keys [[7333.0625,14],[7366.0625,21],[7411.0625,23.1148],[7522.0625,20.4766],[7879.0625,21],[7912.0625,14]]
       }]
     }]
   }
   {
    type=MovieSceneFloatTrack
    display_name="Manual Focus Distance (Focus Settings)"
    sections [{type=MovieSceneFloatSection, channels [None {type=MovieSceneScriptingFloatChannel, default=427.0862}]}]
   }
   {
    type=MovieSceneFloatTrack
    display_name=Current Aperture
    sections [{type=MovieSceneFloatSection, channels [None {type=MovieSceneScriptingFloatChannel, default=2.8}]}]
   }
  ]
 }
 PistolCanary_CinematicNPC_Superhero_Main {
  id=#3
  possessed_class=PistolCanary_CinematicNPC_Superhero_Main_C
  tracks [
   {type=MovieSceneSpawnTrack, sections [{type=MovieSceneSpawnSection, channels [None {type=MovieSceneScriptingBoolChannel, default=true}]}]}
   {
    type=MovieSceneVisibilityTrack
    sections [{
      type=MovieSceneVisibilitySection
      channels [None {type=MovieSceneScriptingBoolChannel, default=false, keys [[7852.06,true],[8364.06,false]]}]
     }]
   }
   {type=MovieScene3DAttachTrack, sections [{type=MovieScene3DAttachSection, end_frame=11496, start_frame=3011}]}
   {
    type=MovieSceneSkeletalAnimationTrack
    sections [
     {type=MovieSceneSkeletalAnimationSection, end_frame=7884, start_frame=7517}
     {type=MovieSceneSkeletalAnimationSection, end_frame=7918, start_frame=7884}
    ]
   }
   {
    type=MovieScene3DTransformTrack
    sections [{
      type=MovieScene3DTransformSection
      channels [
       Location.X {
        type=MovieSceneScriptingDoubleChannel
        default=15116
        thinned_from=9
        keys [[7552.0625,20816.8399],[7552.065,-3198.8197],[7574.0625,-1976.9495],[7589.0625,-1161.5045],[7624.065,-44.6676],[7670.065,-0.2243],[7878.065,-2.9975],[7940.065,-94.0002]]
       }
       Location.Y {
        type=MovieSceneScriptingDoubleChannel
        default=69988
        thinned_from=10
        keys [[7552.0625,77393.4936],[7552.065,10173.0719],[7574.0625,6544.3308],[7589.0625,4027.4133],[7624.065,402.7631],[7670.0625,321.0564],[7878.0675,316.9448],[7930.065,-481.3076]]
       }
       Location.Z {
        type=MovieSceneScriptingDoubleChannel
        default=3520
        thinned_from=9
        keys [[7552.0625,14005.8375],[7552.065,8943.3396],[7574.0625,3559.1111],[7589.0625,1010.0874],[7624.065,86.9542],[7670.065,103.8124],[7878.0675,106.7874],[7930.065,658.9877]]
       }
       Rotation.X {
        type=MovieSceneScriptingDoubleChannel
        default=0
        thinned_from=10
        keys [[7552,27.2469],[7574,27.2469],[7589,26.9262],[7624.0025,-0.2841],[7670.0025,-1.3184],[7887.0025,-1.3198],[7904.0025,-1.4514],[7940.0025,20.5613]]
       }
       Rotation.Y {
        type=MovieSceneScriptingDoubleChannel
        default=0
        thinned_from=9
        keys [[7552,-10.0485],[7574,-10.0485],[7589,-3.1056],[7624.0025,2.7722],[7670.0025,2.6258],[7886.005,2.6526],[7903.005,25.639],[7940.0025,-6.6334]]
       }
       Rotation.Z {
        type=MovieSceneScriptingDoubleChannel
        default=0
        thinned_from=9
        keys [[7552,52.7023],[7552.0025,-161.2977],[7574,-161.2977],[7589,-164.2915],[7624.0025,-172.9302],[7904.0025,-173.0319],[7940.0025,-279.1771]]
       }
       Scale.X {type=MovieSceneScriptingDoubleChannel, default=1, keys [[7552.01,1],[7557.01,1.2],[7896.01,1.2],[7912.01,1]]}
       Scale.Y {type=MovieSceneScriptingDoubleChannel, default=1, keys [[7552.01,1],[7557.01,1.2],[7896.01,1.2],[7912.01,1]]}
       Scale.Z {type=MovieSceneScriptingDoubleChannel, default=1, keys [[7552.01,1],[7557.01,1.2],[7896.01,1.2],[7912.01,1]]}
      ]
     }]
   }
  ]
 }
 moverOffset {
  id=#4
  possessed_class=FortStaticMeshActor
  tracks [
   {type=MovieSceneSpawnTrack, sections [{type=MovieSceneSpawnSection, channels [None {type=MovieSceneScriptingBoolChannel, default=true}]}]}
   {
    type=MovieScene3DTransformTrack
    sections [{
      type=MovieScene3DTransformSection
      channels [
       Location.X {type=MovieSceneScriptingDoubleChannel, default=9692.0309, const [9692.0312,12]}
       Location.Y {type=MovieSceneScriptingDoubleChannel, default=90916.5902, const [90916.5938,12]}
       Location.Z {type=MovieSceneScriptingDoubleChannel, default=9455.7201, const [9455.7197,12]}
       Rotation.X {
        type=MovieSceneScriptingDoubleChannel
        default=-0.5532
        thinned_from=18
        keys [[6845.0625,-0.5532],[7047.0625,6.0468],[7615.0625,0],[8006.0625,0],[8182.0625,8.4],[8632.0625,8.4],[8753.0625,-1.6],[8878.0625,8.4],[9042.0625,8.4],[9183.0625,4.2],[9226.0625,4.2],[9298.0625,19.6],[10318.0625,19.6],[10412.0625,-7.9]]
       }
       Rotation.Y {
        type=MovieSceneScriptingDoubleChannel
        default=-0.4136
        thinned_from=18
        keys [[6845.0625,-0.4136],[7047.0625,-1.6136],[7615.0625,0],[8006.0625,0],[8182.0625,-2.8],[8632.0625,-2.8],[8753.0625,0.5],[8878.0625,-2.8],[9042.0625,-2.8],[9183.0625,-0.8],[9226.0625,-0.8],[9298.0625,-5.4],[10318.0625,-5.4],[10412.0625,2]]
       }
       Rotation.Z {
        type=MovieSceneScriptingDoubleChannel
        default=-151
        thinned_from=20
        keys [[7047.0625,-155.8805],[7615.0625,-153.687],[8006.0625,-153.687],[8182.0625,-159.187],[8439.0625,-159.187],[8632.0625,-151.3],[8878.0625,-159.187],[9226.0625,-159.187],[9298.0625,-133],[9423.0625,-121],[9455.0625,-121],[9526.0625,-154],[9572.0625,-155],[9727.0625,-147.1],[9929.0625,-147.1],[9985.0625,-151],[10318.0625,-151]]
       }
       Scale.X {type=MovieSceneScriptingDoubleChannel, default=1}
       Scale.Y {type=MovieSceneScriptingDoubleChannel, default=1}
       Scale.Z {type=MovieSceneScriptingDoubleChannel, default=1}
      ]
     }]
   }
  ]
 }
 player {
  id=#5
  possessed_class=FortPlayerMannequin
  tracks [
   {type=MovieSceneSpawnTrack, sections [{type=MovieSceneSpawnSection, channels [None {type=MovieSceneScriptingBoolChannel, default=true}]}]}
   {
    type=MovieScene3DTransformTrack
    sections [{
      type=MovieScene3DTransformSection
      channels [
       Location.X {
        type=MovieSceneScriptingDoubleChannel
        default=-0.0001
        keys [[7611.0625,45.8753],[7651.0625,-0.0001],[7884.0625,-0.0001],[7914.0625,45.8753]]
       }
       Location.Y {type=MovieSceneScriptingDoubleChannel, default=0, keys [[7611.0625,72.8107],[7651.0625,0],[7884.0625,0],[7914.0625,72.8107]]}
       Location.Z {
        type=MovieSceneScriptingDoubleChannel
        default=0.0002
        keys [[7611.0625,-19.9055],[7651.0625,0.0002],[7884.0625,0.0002],[7914.0625,-19.9055]]
       }
       Rotation.X {type=MovieSceneScriptingDoubleChannel, default=0, const [0,4]}
       Rotation.Y {type=MovieSceneScriptingDoubleChannel, default=0, const [0,4]}
       Rotation.Z {type=MovieSceneScriptingDoubleChannel, default=0, const [0,4]}
       Scale.X {type=MovieSceneScriptingDoubleChannel, default=1}
       Scale.Y {type=MovieSceneScriptingDoubleChannel, default=1}
       Scale.Z {type=MovieSceneScriptingDoubleChannel, default=1}
      ]
     }]
   }
   {
    type=MovieSceneSkeletalAnimationTrack
    sections [
     {type=MovieSceneSkeletalAnimationSection, end_frame=7884, start_frame=3374}
     {type=MovieSceneSkeletalAnimationSection, end_frame=9426, start_frame=7552}
     {type=MovieSceneSkeletalAnimationSection, end_frame=11031, start_frame=9157}
    ]
   }
   {type=MovieScene3DAttachTrack, sections [{type=MovieScene3DAttachSection, end_frame=11496, start_frame=3011}]}
  ]
 }
 KalEl {id=#6, possessed_class=PistolCanary_CinematicNPC_Superhero_Main_C, tracks []}
 camera_playerpov {
  id=#2
  possessed_class=CineCameraActor
  tracks [
   {
    type=MovieScene3DTransformTrack
    sections [{
      type=MovieScene3DTransformSection
      channels [
       Location.X {
        type=MovieSceneScriptingDoubleChannel
        default=-134.7842
        thinned_from=13
        keys [[5885.0625,-68.4267],[6539.0625,-69.1524],[7268.0625,-69.6651],[7354.0625,-44.1875],[7662.0625,-44.1875],[7716.0625,-49.9666],[7879.065,-44.1875],[7904.065,-82.6975],[8551.0625,-113.1307],[8830.0625,-62.4488],[9021.0625,-35.8984],[9498.0625,-79.5122]]
       }
       Location.Y {
        type=MovieSceneScriptingDoubleChannel
        default=84180.4813
        keys [[5885.0625,-118.997],[6539.0625,-121.8356],[7214.0625,-123.841],[7268.0625,-123.841],[7354.0625,-78.0198],[7662.0625,-78.0198],[7716.0625,-76.4066],[7879.065,-78.0198],[7904.065,-112.4977],[8551.0625,-91.4751],[8830.0625,-53.9217],[9021.0625,-42.8308],[9498.0625,-37.7606]]
       }
       Location.Z {
        type=MovieSceneScriptingDoubleChannel
        default=173.0444
        thinned_from=13
        keys [[5885.0625,166.9624],[7268.0625,167.0002],[7354.0625,171.8148],[7662.0625,180.2148],[7716.0625,178.2148],[7879.065,180.2148],[7904.065,181.3002],[8551.0625,175.2083],[8830.0625,164.6682],[9021.0625,158.9206],[9498.0625,173.0444]]
       }
       Rotation.X {
        type=MovieSceneScriptingDoubleChannel
        default=0
        thinned_from=11
        keys [[6539.0625,0],[7904.065,0],[8551.0625,-0.9158],[8644.0625,-0.8999],[9021.0625,-0.6065],[9498.0625,-3.3989]]
       }
       Rotation.Y {
        type=MovieSceneScriptingDoubleChannel
        default=-7.4961
        keys [[6539.0625,4.0282],[7256.0625,11.0282],[7310.0625,4.0282],[7354.0625,1.5282],[7662.0625,3.4282],[7716.0625,3.8282],[7879.065,3.4282],[7904.065,-1.5718],[8551.0625,0.7972],[8644.0625,0.9034],[8722.0625,-7.4961],[8799.0625,-2.4631],[9021.0625,5.068],[9498.0625,4.0197]]
       }
       Rotation.Z {
        type=MovieSceneScriptingDoubleChannel
        default=76.2004
        keys [[6539.0625,75.7004],[7214.0625,76.2004],[7268.0625,75.7004],[7354.0625,76.5004],[7662.0625,76.5004],[7716.0625,74.7004],[7879.065,76.5004],[7904.065,75.1004],[8551.0625,71.3191],[8644.0625,72.0399],[9021.0625,78.2391],[9498.0625,66.2647]]
       }
       Scale.X {type=MovieSceneScriptingDoubleChannel, default=1, const [1,5]}
       Scale.Y {type=MovieSceneScriptingDoubleChannel, default=1, const [1,5]}
       Scale.Z {type=MovieSceneScriptingDoubleChannel, default=1, const [1,5]}
      ]
     }]
   }
   {type=MovieSceneSpawnTrack, sections [{type=MovieSceneSpawnSection, channels [None {type=MovieSceneScriptingBoolChannel, default=true}]}]}
   {
    type=TemplateSequenceTrack
    display_name=Template Animation
    sections [
     {
      type=TemplateSequenceSection
      end_frame=9487
      start_frame=7244
      channels ["Transform[Rotation]" {type=MovieSceneScriptingFloatChannel, default=0.1, keys [[7552.0625,0],[7612.0625,0.1],[7864.0625,0.1],[7914.0625,0]]}]
     }
     {
      type=TemplateSequenceSection
      end_frame=7956
      start_frame=7866
      channels ["Transform[Rotation]" {type=MovieSceneScriptingFloatChannel, default=0, keys [[7866.0625,0],[7888.0625,0.21],[7914.0625,0]]}]
     }
     {
      type=TemplateSequenceSection
      end_frame=7956
      start_frame=7873
      channels ["Transform[Rotation]" {type=MovieSceneScriptingFloatChannel, default=0, keys [[7873.0625,0],[7888.0625,0.23],[7914.0625,0]]}]
     }
     {
      type=TemplateSequenceSection
      end_frame=7937
      start_frame=7880
      channels ["Transform[Rotation]" {type=MovieSceneScriptingFloatChannel, default=1, keys [[7880.0625,0],[7888.0625,1],[7914.0625,0]]}]
     }
     {
      type=TemplateSequenceSection
      end_frame=7678
      start_frame=7569
      channels ["Transform[Rotation]" {type=MovieSceneScriptingFloatChannel, default=0, keys [[7569.0625,0],[7605.0625,0.08],[7647.645,0.08],[7676.3725,0]]}]
     }
     {
      type=TemplateSequenceSection
      end_frame=10421
      start_frame=5886
      channels ["Transform[Rotation]" {type=MovieSceneScriptingFloatChannel, default=0.2}]
     }
     {
      type=TemplateSequenceSection
      end_frame=7133
      start_frame=7024
      channels ["Transform[Rotation]" {type=MovieSceneScriptingFloatChannel, default=0, keys [[7024.0625,0],[7060.0625,0.08],[7102.645,0.08],[7131.3725,0]]}]
     }
     {
      type=TemplateSequenceSection
      end_frame=8979
      start_frame=8870
      channels ["Transform[Rotation]" {type=MovieSceneScriptingFloatChannel, default=0, keys [[8870.0625,0],[8906.0625,0.08],[8948.645,0.08],[8977.3725,0]]}]
     }
     {type=TemplateSequenceSection, end_frame=7340, start_frame=7245}
     {
      type=TemplateSequenceSection
      end_frame=7569
      start_frame=7460
      channels ["Transform[Rotation]" {type=MovieSceneScriptingFloatChannel, default=0, keys [[7460.0625,0],[7475.0625,0.08],[7532.0625,0.08],[7543.3725,0]]}]
     }
    ]
   }
   {type=MovieScene3DAttachTrack, sections [{type=MovieScene3DAttachSection, end_frame=11253, start_frame=2768}]}
  ]
 }
 PistolCanary_BP_Countdown {
  id=#7
  possessed_class=BP_Generic_Game_Countdown_C
  tracks [{
    type=MovieSceneVisibilityTrack
    sections [{type=MovieSceneVisibilitySection, channels [None {type=MovieSceneScriptingBoolChannel, default=false}]}]
   }]
 }
]
display_rate {denominator=1, numerator=60}
tick_resolution {denominator=1, numerator=24000}
tracks [
 {type=MovieSceneSubTrack, display_name=PistolCanary - Main B, sections [{type=MovieSceneSubSection, end_frame=10148, start_frame=4453}]}
 {type=MovieSceneSubTrack, display_name=moc - moment of calm, sections [{type=MovieSceneSubSection, end_frame=4453, start_frame=3600}]}
 {
  type=MovieSceneDataLayerTrack
  display_name=Data Layer - InnerSpire4
  sections [{type=MovieSceneDataLayerSection, end_frame=10148, start_frame=9019}]
 }
 {
  type=MovieSceneDataLayerTrack
  display_name=Data Layer - Tentacle
  sections [{type=MovieSceneDataLayerSection, end_frame=10148, start_frame=7029}, {type=MovieSceneDataLayerSection, end_frame=7029, start_frame=3600}]
 }
 {
  type=MovieSceneDataLayerTrack
  display_name=Data Layer - S3_to_MatchEnd
  sections [{type=MovieSceneDataLayerSection, end_frame=10148, start_frame=3600}]
 }
 {
  type=MovieSceneDataLayerTrack
  display_name="Data Layer - S2_to_MatchEnd (Section Skip)"
  sections [{type=MovieSceneDataLayerSection, end_frame=10148, start_frame=3600}]
 }
 {
  type=MovieSceneDataLayerTrack
  display_name="Data Layer - S1_to_End_S3 (Section Skip)"
  sections [{type=MovieSceneDataLayerSection, end_frame=10148, start_frame=3600}]
 }
 {type=MovieSceneDataLayerTrack, display_name=Data Layer - S3, sections [{type=MovieSceneDataLayerSection, end_frame=10148, start_frame=3600}]}
 {type=MovieSceneDataLayerTrack, display_name=Data Layer - S2, sections [{type=MovieSceneDataLayerSection, end_frame=10148, start_frame=3600}]}
 {
  type=MovieSceneDataLayerTrack
  display_name=Data Layer - InnerSpire3
  sections [{type=MovieSceneDataLayerSection, end_frame=10148, start_frame=9176}]
 }
 {
  type=MovieSceneDataLayerTrack
  display_name=Data Layer - InnerSpire2
  sections [{type=MovieSceneDataLayerSection, end_frame=10148, start_frame=9432}]
 }
 {
  type=MovieSceneDataLayerTrack
  display_name=Data Layer - InnerSpire1
  sections [{type=MovieSceneDataLayerSection, end_frame=10148, start_frame=9313}]
 }
]
